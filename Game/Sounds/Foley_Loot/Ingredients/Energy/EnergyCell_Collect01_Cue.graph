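# Sound cue for collecting an Energy Cell ingredient, part of the loot foley set.
# A mixer layers two wave players, one "Real" and one "Surreal" collect sound, played together at equal volume.
# It uses the prop-collect attenuation and the player-interacts sound class.
# A concurrency override caps it at two simultaneous instances, stopping the oldest.
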SoundCue /Game/Sounds/Foley_Loot/Ingredients/Energy/EnergyCell_Collect01_Cue
root=Mixer_7
nodes [
 {id=Mixer_7, children [WavePlayer_16, WavePlayer_15], properties {inputVolume [1,1]}}
 {id=WavePlayer_15, properties {soundWaveAssetPtr=&/EnergyCell_Collect_Surreal}}
 {id=WavePlayer_16, properties {soundWaveAssetPtr=&/EnergyCell_Collect_Real}}
]
properties {
 attenuationSettings=/Game/Sounds/Attenuation/Fort_Prop_Collect_Att
 bOverrideConcurrency=true
 maxDistance=2097152
 soundClassObject=/Game/Sounds/FortSoundClassesAndModes/Fort_08_Player_Interacts
 concurrencyOverrides {maxCount=2, resolutionRule=StopOldest}
}
