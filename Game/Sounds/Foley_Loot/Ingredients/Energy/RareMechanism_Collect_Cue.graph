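# Sound cue for collecting a Rare Mechanism, filed under energy ingredients in the loot foley sounds.
# A mixer at the root plays two wave layers together, a "Real" and a "Surreal" collect sound, at slightly different input volumes.
# The cue uses the shared prop-collect attenuation and loot-collect sound class at a reduced volume.
# A concurrency override caps simultaneous instances at a small count, stopping the oldest first.
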SoundCue /Game/Sounds/Foley_Loot/Ingredients/Energy/RareMechanism_Collect_Cue
root=Mixer_19
nodes [
 {id=Mixer_19, children [WavePlayer_38, WavePlayer_39], properties {inputVolume [0.8,1]}}
 {id=WavePlayer_38, properties {soundWaveAssetPtr=&/RareMech_Collect_Real}}
 {id=WavePlayer_39, properties {soundWaveAssetPtr=&/RareMech_Collect_Surreal}}
]
properties {
 attenuationSettings=/Game/Sounds/Attenuation/Fort_Prop_Collect_Att
 bOverrideConcurrency=true
 maxDistance=2097152
 soundClassObject=&Fort_Loot_Collect
 volumeMultiplier=0.299
 concurrencyOverrides {maxCount=2, resolutionRule=StopOldest}
}
