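# Sound cue for collecting a rare powercell energy ingredient as loot.
# A mixer layers two waves, a real and a surreal collect sound, at equal volume so they play together.
# The cue lowers overall volume and uses the shared prop-collect attenuation and loot-collect sound class.
# A concurrency override caps simultaneous plays at a small count, stopping the oldest when exceeded.
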SoundCue /Game/Sounds/Foley_Loot/Ingredients/Energy/Rare_Powercell_Collect_Cue
root=Mixer_5
nodes [
 {id=Mixer_5, children [WavePlayer_11, WavePlayer_12], properties {inputVolume [1,1]}}
 {id=WavePlayer_11, properties {soundWaveAssetPtr=&/Powercell_Collect_Real}}
 {id=WavePlayer_12, properties {soundWaveAssetPtr=&/Powercell_Collect_Surreal}}
]
properties {
 attenuationSettings=/Game/Sounds/Attenuation/Fort_Prop_Collect_Att
 bOverrideConcurrency=true
 maxDistance=2097152
 soundClassObject=&Fort_Loot_Collect
 volumeMultiplier=0.259
 concurrencyOverrides {maxCount=2, resolutionRule=StopOldest}
}
